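; Pickaxe weapon Blueprint for the EggnogFaucetUniversal cosmetic, driving its swing and trail effects.
; On equip it binds the instigator pawn's skeletal mesh to the swing Niagara component's "Pawn" parameter.
; Each left or right swing snaps the anim trail to the two-handed handle socket name and sets "Swing02Delay" when the active montage section is Swing2.
; Construction turns off auto-activation of the idle particle system.
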(event FortWeapon|EventOnEquippedAndAttached
  (|CallParentFunction)
  (Utilities|FlowControl|Sequence
    (:then_0
      (bind _asfort_pawn (Utilities|Casting|CastToFortPawn (Game|GetInstigator)))
      (Utilities|IsValid (Variables|/Script/Engine.Character|GetMesh _asfort_pawn)
        (:Is%20Valid
          (bind _asniagara_particle_system_component (Utilities|Casting|CastToNiagaraComponent (CallFunction|GetSwingVFXComponent))
            (:then
              (Niagara|SetNiagaraSkeletalMeshComponent _asniagara_particle_system_component "Pawn" (Variables|/Script/Engine.Character|GetMesh _asfort_pawn)))
            (:CastFailed
              (Development|PrintString "CAST FAILED - Niagara Component"))))))))

(event EventSwingLeft
  (bind _returnvalue (CallFunction|GetAnimTrailVFXComponent))
  (|CallParentFunction)
  (Utilities|FlowControl|Sequence
    (:then_0
      (Transformation|AttachComponentToComponent _returnvalue 0 "Melee_TwoHanded_Handle" "SnapToTarget" "SnapToTarget" "SnapToTarget" true))
    (:then_1
      (Effects|Components|ParticleSystem|SetBoolParameter _returnvalue "Swing02Delay" (== (Animation|Montage|MontageGetCurrentSection (Components|SkeletalMesh|GetAnimInstance (Variables|/Script/Engine.Character|GetMesh (Utilities|Casting|CastToFortPlayerPawn (Game|GetInstigator)))) (Animation|Montage|GetCurrentActiveMontage (Components|SkeletalMesh|GetAnimInstance (Variables|/Script/Engine.Character|GetMesh (Utilities|Casting|CastToFortPlayerPawn (Game|GetInstigator)))))) "Swing2")))))

(event EventSwingRight
  (bind _returnvalue (CallFunction|GetAnimTrailVFXComponent))
  (|CallParentFunction)
  (Utilities|FlowControl|Sequence
    (:then_0
      (Transformation|AttachComponentToComponent _returnvalue 0 "Melee_TwoHanded_Handle" "SnapToTarget" "SnapToTarget" "SnapToTarget" true))
    (:then_1
      (Effects|Components|ParticleSystem|SetBoolParameter _returnvalue "Swing02Delay" (== (Animation|Montage|MontageGetCurrentSection (Components|SkeletalMesh|GetAnimInstance (Variables|/Script/Engine.Character|GetMesh (Utilities|Casting|CastToFortPlayerPawn (Game|GetInstigator)))) (Animation|Montage|GetCurrentActiveMontage (Components|SkeletalMesh|GetAnimInstance (Variables|/Script/Engine.Character|GetMesh (Utilities|Casting|CastToFortPlayerPawn (Game|GetInstigator)))))) "Swing2")))))

(fn ConstructionScript
  (|CallParentFunction)
  (Components|Activation|SetAutoActivate (CallFunction|GetIdlePSC) false))
