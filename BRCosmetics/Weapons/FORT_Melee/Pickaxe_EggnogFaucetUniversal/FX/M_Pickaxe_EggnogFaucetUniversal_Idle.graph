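# Translucent unlit Niagara mesh-particle material for the EggnogFaucetUniversal pickaxe idle effect.
# Opacity combines panning noise and V gradients with particle alpha, view-angle fade and depth fade.
# Emissive uses a fire curve atlas to blend two colors, modulated by noise and particle color, with a TwoSidedSign-selected backside tint; both colors take green from their red channel.
# A panning texture, masked toward mid-V, feeds a transformed world position offset.
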Material /BRCosmetics/Weapons/FORT_Melee/Pickaxe_EggnogFaucetUniversal/FX/M_Pickaxe_EggnogFaucetUniversal_Idle
graph {
 properties {
  bUsedWithNiagaraMeshParticles=true
  blendMode=BLEND_Translucent
  pixelDepthOffsetMode=PDOM_Legacy
  shadingModel=MSM_Unlit
  usedShadingModels=Unlit
 }
 nodes [
  n1 = ParticleColor()
  n2 = Panner(Coordinate=TextureCoordinate() {uTiling=0.5, vTiling=3}) {speedX=-0.5, speedY=0.5}
  n3 = Power(Base=Multiply(A=TextureSample(UVs=@n2) {texture=/BRCosmetics/Weapons/FORT_Melee/Pickaxe_CoatCheck/FX/T_Pickaxe_CoatCheck_Noise}.R) {constB=2})
  n4 = Multiply(A=ComponentMask(None=TextureCoordinate()) {g=true}) {constB=2}
  n5 = ComponentMask(None=Add(A=TextureCoordinate(), B=Constant2Vector() {g=-0.25})) {g=true}
  n6 = Saturate(None=Add(A=Multiply(A=@n3, B=@n4), B=Multiply(A=@n5) {constB=2}))
  n7 = ScalarParameter() {defaultValue=2, parameterName=AngleFade}
  DepthFade(
   Opacity=Multiply(
    A=Multiply(A=@n1.A, B=@n6)
    B=MaterialFunctionCall("FadeSensitivity (optional)"=@n7) {materialFunction=&FadeBasedOnViewAngle}.Result
   )
   FadeDistance=ScalarParameter() {defaultValue=16, parameterName=FadeDistance}
  )
  n8 = TextureCoordinate()
  n9 = Constant2Vector() {r=-0.25}
  n10 = CurveAtlasRowParameter(CurveTime=ComponentMask(None=Add(A=@n8, B=@n9)) {r=true}) {
   atlas=/Game/Athena/LinearColorCurves/TempAthenaCurveAtlas
   curve=/Game/Weapons/FORT_Melee/Pickaxe_Elastic/Fire/Curve_Pickaxe_Elastic_Fire_01
   defaultValue=14
  }
  n11 = VectorParameter() {group=Color, parameterName=Secondary Color, defaultValue=<rgba 1 0.05 0 1>}
  n12 = AppendVector(A=AppendVector(A=AppendVector(A=@n11.R, B=@n11.R), B=@n11.B), B=@n11.A)
  n13 = Multiply(A=@n12, B=ScalarParameter() {defaultValue=0.08, group=Color, parameterName=Backside Color Intensity})
  n14 = ScalarParameter() {defaultValue=0.3, group=Color, parameterName=Backside Color Desaturation}
  n15 = TextureCoordinate()
  n16 = AppendVector(
   A=ScalarParameter() {defaultValue=1.5, group=Noise, parameterName=NoiseTexture_Tile_U}
   B=ScalarParameter() {defaultValue=0.5, group=Noise, parameterName=NoiseTexture_Tile_V}
  )
  n17 = ScalarParameter() {group=Noise, parameterName=NoiseTexture_Pan_U}
  n18 = ScalarParameter() {group=Noise, parameterName=NoiseTexture_Pan_V}
  n19 = TextureSampleParameter2D(UVs=Panner(Coordinate=Multiply(A=@n15, B=@n16), Speed=AppendVector(A=@n17, B=@n18)) {speedY=0.3}) {group=Noise, parameterName=NoiseTexture, texture=/Game/Packages/Fortress_Transylvania/Haunted_Forest/Swamps/Textures/2ndInfectedMapLinear2}
  n20 = Multiply(
   A=Power(Base=@n19.R, Exp=ScalarParameter() {defaultValue=1, group=Noise, parameterName=NoiseTexture_Power})
   B=ScalarParameter() {defaultValue=1, group=Noise, parameterName=NoiseTexture_Intensity}
  )
  n21 = Multiply(A=Multiply(A=@n10, B=Desaturation(None=@n13, Fraction=@n14)), B=@n20)
  n22 = Desaturation(
   None=Multiply(A=@n12, B=ScalarParameter() {defaultValue=1, group=Color, parameterName=Secondary Color Intensity})
   Fraction=ScalarParameter() {defaultValue=0.2, group=Color, parameterName=Secondary Color Desaturation}
  )
  n23 = VectorParameter() {group=Color, parameterName=Primary Color, defaultValue=<rgba 1 0.5 0 1>}
  n24 = AppendVector(A=AppendVector(A=@n23.R, B=@n23.R), B=@n23.B)
  n25 = Desaturation(
   None=Multiply(A=AppendVector(A=@n24, B=@n23.A), B=ScalarParameter() {defaultValue=1, group=Color, parameterName=Primary Color Intensity})
   Fraction=ScalarParameter() {defaultValue=0.15, group=Color, parameterName=Primary Color Desaturation}
  )
  n26 = TextureCoordinate()
  n27 = Constant2Vector() {r=-0.25}
  n28 = CurveAtlasRowParameter(CurveTime=ComponentMask(None=Add(A=@n26, B=@n27)) {r=true}) {
   atlas=&TempAthenaCurveAtlas
   curve=&Curve_Pickaxe_Elastic_Fire_01
   defaultValue=14
  }
  n29 = Multiply(A=Multiply(A=LinearInterpolate(A=@n22, B=@n25, Alpha=@n28.R), B=@n28.G), B=@n20)
  n30 = Clamp(None=TwoSidedSign())
  MaterialFunctionCall(Emissive=Multiply(A=LinearInterpolate(A=@n21, B=@n29, Alpha=@n30), B=@n1.RGB)) {materialFunction=&MF_FX_EmissiveAdjustments}
  n31 = ScalarParameter() {defaultValue=1, group=WPO, parameterName=WPO_Intensity}
  n32 = TextureCoordinate()
  n33 = AppendVector(
   A=ScalarParameter() {defaultValue=2, group=WPO, parameterName=WPOTexture_Tile_U}
   B=ScalarParameter() {defaultValue=0.5, group=WPO, parameterName=WPOTexture_Tile_V}
  )
  n34 = ScalarParameter() {group=WPO, parameterName=WPOTexture_Pan_U}
  n35 = ScalarParameter() {defaultValue=-1, group=WPO, parameterName=WPOTexture_Pan_V}
  n36 = TextureSampleParameter2D(UVs=Panner(Coordinate=Multiply(A=@n32, B=@n33), Speed=AppendVector(A=@n34, B=@n35)) {speedY=0.3}) {group=WPO, parameterName=WPO_Texture, texture=&2ndInfectedMapLinear}
  n37 = TextureCoordinate()
  n38 = Multiply(A=ComponentMask(None=@n37) {g=true}, B=OneMinus(None=ComponentMask(None=@n37) {g=true}))
  Transform(None=Multiply(A=Multiply(A=@n31, B=@n36.RGB), B=Multiply(A=@n38) {constB=2}))
 ]
}
