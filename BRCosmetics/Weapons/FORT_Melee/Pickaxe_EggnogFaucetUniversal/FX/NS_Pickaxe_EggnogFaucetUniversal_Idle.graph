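# Niagara idle effect for the EggnogFaucetUniversal pickaxe: a fiery, glowing orange ring-and-core look.
# Persistent burst emitters hold the center, two counter-rotating mirrored rings, an extra ring, trail mesh, glow sprites and light.
# Continuous emitters add fire meshes, drifting flame sprites and sparks swirled by vortex and attraction forces.
# User.On_BuiltIn blends placement, orientation and scale between two layouts, while time-of-day, front-end and platform modifiers adjust some brightness.
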NiagaraSystem /BRCosmetics/Weapons/FORT_Melee/Pickaxe_EggnogFaucetUniversal/FX/NS_Pickaxe_EggnogFaucetUniversal_Idle
dependencies {
 usedDataInterfaces [
  &NiagaraDataInterfaceArrayFloat3
  &NiagaraDataInterfaceArrayFloat
  &NiagaraDataInterfacePlatformSet
  &NiagaraDataInterfaceCurve
  &NiagaraDataInterfaceColorCurve
 ]
 usedDynamicInputs []
 usedModules [&SystemState]
}
emitters {
 Center {
  emitterUpdateScript [
   &EmitterState(
    Inactive Response=::NewEnumerator1
    Life Cycle Mode=::NewEnumerator1
    Loop Behavior=::NewEnumerator1
    Loop Duration Mode=::NewEnumerator1
    UseLoopDelay=false
   )
   &SpawnBurst_Instantaneous(Use Spawn Probability=true, UseLoopCountLimit=false)
  ]
  particleSpawnScript [
   &InitializeParticle(
    Color Mode=::NewEnumerator0
    Lifetime=5
    Mesh Scale Mode=::NewEnumerator4
    Position Offset=&Lerp_Vector(A=<0 -5 0>, B=<15 2.5 -3>, Alpha=$User.On_BuiltIn)
    UsePositionOffset=true
    Write Material Random=true
   )
   &InitialMeshOrientation(
    Mesh Orientation Mode=::NewEnumerator2
    Rotation=&Lerp_Vector(A=<0 0 0>, B=<0 0.02 0.04>, Alpha=$User.On_BuiltIn)
    Use Rotation Vector=true
   )
  ]
  particleUpdateScript [
   &ParticleState(Kill Particles When Lifetime Has Elapsed=false, Let Infinitely Lived Particles Die When Emitter Deactivates=true)
   &UpdateMeshOrientation(Rotation Rate=0.65)
   /SetVariables(Particles.LightColor=&Multiply_LinearColorByFloat(
     Linear Color=<rgba 0.7604 0.4791 0.0683 0.4>
     Float=&Multiply_Float(
      A=&FNTimeOfDayScalar(Night Scale=0.85, Morning Scale=1, Evening Scale=0.8)
      B=&FN_FrontEndFloatModifier(
       Gameplay=&FN_PickFloatDependOnPlatform(TargetPlatformActive=1, TargetPlatformNoActive=0.85)
       Front End=0.88
      )
     )
    )) {isSetParameters=true}
   /SetVariables(Particles.LightSize:Vector3f=<2 1.2 1.2>) {isSetParameters=true}
   /SetVariables(Particles.DarkColor:LinearColor=<rgba 0.4896 0.0443 0 2.5>) {isSetParameters=true}
   /SetVariables(Particles.DarkSize:Vector3f=<2.4 0.85 0.85>) {isSetParameters=true}
   /SetVariables(Particles.DarkMask_Color:LinearColor=<rgba 0.1562 0.0141 0 4>) {isSetParameters=true}
   /SetVariables(Particles.DarkMask_Size:Vector3f=&Lerp_Vector(A=<2 0.3 0.3>, B=<0 0 0>, Alpha=$User.On_BuiltIn)) {isSetParameters=true}
   &FN_Sprite_Attributes(
    Color=&Multiply_LinearColorByFloat(
     Linear Color=<rgba 1 0.15 0.03 0.8>
     Float=&Multiply_Float(
      A=&FNTimeOfDayScalar(Night Scale=0.7, Morning Scale=0.8, Evening Scale=0.8)
      B=&FN_FrontEndFloatModifier(Gameplay=1, Front End=0.75)
     )
    )
    Create Color Attribute=true
    Create Position Attribute=true
    Create SpriteSize Attribute=true
    Position:NiagaraPosition=&Lerp_Vector(A=<-8 -5 0>, B=<5 0 -2>, Alpha=$User.On_BuiltIn)
    SpriteSize=&Vector2DFromFloat(Value=50)
    Write to Module Namespace=true
   ) {module=FN_Sprite_Attributes001}
  ]
  rendererClasses [&NiagaraMeshRendererProperties, &NiagaraSpriteRendererProperties]
  renderers [
   {rendererClass=&NiagaraMeshRendererProperties, rendererIndex=0}
   {rendererClass=&NiagaraMeshRendererProperties, rendererIndex=1}
   {rendererClass=&NiagaraSpriteRendererProperties, rendererIndex=2}
   {rendererClass=&NiagaraMeshRendererProperties, rendererIndex=3}
  ]
 }
 FireRing_Limbs001 {
  emitterUpdateScript [
   &EmitterState(Inactive Response=::NewEnumerator1, Life Cycle Mode=::NewEnumerator1, Loop Duration=1, UseLoopDelay=false)
   &SpawnRate(SpawnRate=&Lerp_Float(A=5, B=6, Alpha=$User.On_BuiltIn), Use Spawn Probability=false)
  ]
  particleSpawnScript [
   &InitializeParticle(
    Lifetime Max=2.5
    Lifetime Min=2
    Lifetime Mode=::NewEnumerator1
    Mesh Scale=&Multiply_Vector(A=<0.9 1.05 1.05>, B=<1 1.05 1.05>)
    Mesh Scale Mode=::NewEnumerator4
    Position Offset=&Lerp_Vector(A=<-3 -5 0>, B=<15 -2 -3>, Alpha=$User.On_BuiltIn)
    UsePositionOffset=true
    Write Material Random=true
   )
   &ShapeLocation(
    Cylinder Height=1
    Cylinder Radius=1
    MaskSpawnGroup=false
    Override Seed=false
    Rotation Mode=::NewEnumerator1
    Shape Primitive=::NewEnumerator1
    Surface Only=false
    UseOwnerScale=true
   )
   &InitialMeshOrientation(
    Mesh Orientation Mode=::NewEnumerator2
    Rotation=&RandomRangeVector(Minimum=<0 0 0>, Maximum=<1 0 0>)
    Use Rotation Vector=true
   )
   &InitialMeshOrientation(Mesh Orientation Mode=::NewEnumerator2, Use Rotation Vector=true) {module=InitialMeshOrientation001}
   &CameraOffset(Camera Offset Amount=5)
  ]
  particleUpdateScript [
   &ParticleState()
   &ScaleMeshSize(Scale Factor=&RandomRangeVector(Minimum=<1 1 1>, Maximum=<1.3 1 1>))
   &ScaleMeshSize(Scale Factor=&Lerp_Vector(A=<1 1 1>, B=<1.15 1.03 1.03>, Alpha=$User.On_BuiltIn)) {module=ScaleMeshSize001}
   &UpdateMeshOrientation(Rotation Coordinate Space=::NewEnumerator0, Rotation Rate=0.1)
   &ScaleColor(
    Scale Alpha=&FloatFromCurve(FloatCurve=DataInterface() {propertyValues {Curve {keys [<at 0 0 RCIM_Cubic>,<at 0.3 1 RCIM_Cubic>,<at 1 0 RCIM_Cubic>]}}})
    Scale RGB=&MakeVectorFromLinearColorRGB(Color=&RandomRangeLinearColor(Minimum=<rgba 1 0.1409 0.065 1>, Maximum=<rgba 1.5 1.2628 0 1>))
    ScaleA=true
    ScaleRGB=true
   )
   &DynamicMaterialParameters(
    Index 0 Param 1=&RandomRangeFloat(Minimum=1, Maximum=2)
    Index 0 Param 2=1
    Index 0 Param 3=3
    Index 0 Param 4=&RandomRangeFloat(Minimum=1, Maximum=1.6)
    Index 2 Param 1=&FloatFromCurve(
     FloatCurve=DataInterface() {propertyValues {Curve {keys [
         {arriveTangent=-0.01, interpMode=RCIM_Cubic, leaveTangent=-0.01, tangentMode=RCTM_User, time=0, value=0}
         {arriveTangent=2.1454, interpMode=RCIM_Cubic, leaveTangent=2.1454, tangentMode=RCTM_User, time=0.2, value=0.2}
         {arriveTangent=0.0029, interpMode=RCIM_Cubic, leaveTangent=0.0029, tangentMode=RCTM_User, time=0.3, value=0.4}
         {arriveTangent=-0.5714, interpMode=RCIM_Cubic, leaveTangent=-0.5714, time=0.5, value=0.1}
         {arriveTangent=-0.0738, interpMode=RCIM_Cubic, leaveTangent=-0.0738, tangentMode=RCTM_User, time=1, value=0}
        ]}}}
     Scale Curve=&Multiply_Float(A=1, B=2)
    )
    Index 2 Param 2=1
    Index 2 Param 3=1
    Index 2 Param 4=1
    Module.Index0Float0Write=true
    Module.Index0Float1Write=true
    Module.Index0Float2Write=true
    Module.Index0Float3Write=true
    Module.Index2Float0Write=true
    Module.Index2Float1Write=true
    Module.Index2Float2Write=true
    Module.Index2Float3Write=true
    Write Parameter Index 2=::NewEnumerator1
   )
  ]
  rendererClasses [&NiagaraMeshRendererProperties]
  renderers [{rendererClass=&NiagaraMeshRendererProperties, rendererIndex=0}]
 }
 Fire_Particle001 {
  emitterUpdateScript [
   &EmitterState(Life Cycle Mode=::NewEnumerator1, Loop Duration=1, UseLoopDelay=false)
   &SpawnRate(SpawnRate=7, Use Spawn Probability=false)
  ]
  particleSpawnScript [
   &InitializeParticle(
    Color=<rgba 0.2656 0.0751 0.054 3>
    Lifetime=1.6
    Position Offset=&Lerp_Vector(A=<-10 -5 0>, B=<13 1 -2>, Alpha=$User.On_BuiltIn)
    Sprite Rotation Mode=::NewEnumerator2
    Sprite Size Mode=::NewEnumerator3
    Uniform Sprite Size=22
    UsePositionOffset=true
    Write Material Random=true
   )
   &AddVelocity(Velocity=&Lerp_Vector(A=<-50 0 0>, B=<-60 -15 5>, Alpha=$User.On_BuiltIn))
   &CameraOffset(Camera Offset Amount=&Lerp_Float(A=0, B=8, Alpha=$User.On_BuiltIn))
  ]
  particleUpdateScript [
   &ParticleState()
   &ScaleSpriteSize(Scale Factor=&Vector2DFromFloat(Value=&FloatFromCurve(FloatCurve=DataInterface() {propertyValues {Curve {keys [<at 0 1>, {arriveTangent=-0.004, time=0.2, value=1}, {arriveTangent=-0.3923, leaveTangent=-0.6919, time=1, value=0.45}]}}})))
   &ScaleColor(
    Scale Alpha=&FloatFromCurve(FloatCurve=DataInterface() {propertyValues {Curve {keys [
         <at 0 0 RCIM_Cubic>
         {arriveTangent=-0.1097, interpMode=RCIM_Cubic, leaveTangent=-0.1097, tangentMode=RCTM_User, time=0.15, value=1}
         {arriveTangent=-1.7326, interpMode=RCIM_Cubic, leaveTangent=-1.7326, tangentMode=RCTM_User, time=0.5, value=0.65}
         <at 1 0 RCIM_Cubic>
        ]}}})
    Scale RGB=&MakeVectorFromLinearColorRGB(Color=&ColorFromCurve(ColorCurve=DataInterface() {propertyValues {
        AlphaCurve {keys [<at 0 1 RCIM_Cubic>]}
        BlueCurve {keys [<at 0.0729 0.029 RCIM_Cubic>,<at 0.6936 0.1287 RCIM_Cubic>]}
        GreenCurve {keys [<at 0.0729 0.1789 RCIM_Cubic>,<at 0.6936 1.2117 RCIM_Cubic>]}
        RedCurve {keys [<at 0.0729 1 RCIM_Cubic>,<at 0.6936 2 RCIM_Cubic>]}
       }}))
    ScaleA=true
    ScaleRGB=true
   )
   &Drag(Use Linear Drag=true, Use Rotational Drag=true)
   &SolveForcesAndVelocity(Clamp Velocity=false, Limit Acceleration=false)
   /SetVariables(Particles.SubImageIndex=&FloatFromCurve(FloatCurve=DataInterface() {propertyValues {Curve {keys [<at 0 5>,<at 1 10>]}}})) {isSetParameters=true}
  ]
  rendererClasses [&NiagaraSpriteRendererProperties]
  renderers [{rendererClass=&NiagaraSpriteRendererProperties}]
 }
 Particle001 {
  emitterUpdateScript [
   &EmitterState(Life Cycle Mode=::NewEnumerator1, Loop Duration=1, UseLoopDelay=false)
   &SpawnRate(SpawnRate=25, Use Spawn Probability=false)
  ]
  particleSpawnScript [
   &InitializeParticle(
    Color=<rgba 5 4.2184 0.7031 1>
    Lifetime Max=1
    Lifetime Min=0.65
    Lifetime Mode=::NewEnumerator1
    Sprite Size Max=<1 4>
    Sprite Size Min=<0.5 1>
    Sprite Size Mode=::NewEnumerator2
    UsePositionOffset=true
    Write Material Random=true
   )
   &ShapeLocation(
    Cylinder Height=30
    Cylinder Height Midpoint=0
    Cylinder Radius=6
    MaskSpawnGroup=false
    Offset=&Lerp_Vector(A=<-25 -5 0>, B=<-10 -3 -2>, Alpha=$User.On_BuiltIn)
    Offset Mode=::NewEnumerator0
    Override Seed=false
    Rotation Mode=::NewEnumerator1
    Shape Primitive=::NewEnumerator1
    Surface Only=true
    Use Endcaps In Surface Only Mode=false
    UseOwnerScale=true
   )
   &AddVelocity(Velocity=<-1 0 0>)
  ]
  particleUpdateScript [
   &ParticleState()
   /SetVariables(Particles.SizeCustomScale:Vector2f=&Multiply_Vector2D_ByFloat(
     Float=&FloatFromCurve(FloatCurve=DataInterface() {propertyValues {Curve {keys [<at 0 0 RCIM_Cubic>, {arriveTangent=0.2, interpMode=RCIM_Cubic, leaveTangent=0.2, time=0.2, value=1}, <at 1 0.2>]}}})
     Vector2D=<0.85 1>
    )) {isSetParameters=true}
   &CurlNoiseForce(Bias Noise=false, Bias Noise Field:Vector3f=<0 0 0.2>, Noise Strength=50, Pan Noise=false) {enabled=false}
   &VortexForce(
    UseOriginOffset=true
    Vortex Axis=<1 0 0>
    Vortex Force Amount=&Lerp_Float(A=300, B=200, Alpha=$User.On_BuiltIn)
    Vortex Origin Offset:Vector3f=<0 -5 0>
   )
   &PointAttractionForce(
    Attraction Radius=50
    Attraction Strength=30
    Attractor Position Offset:Vector3f=&Lerp_Vector(A=<-35 0 0>, B=<-30 -5 0>, Alpha=$User.On_BuiltIn)
    Kill Within Radius=false
    Module.Use Falloff=true
    UseAttractorOffset=true
   )
   &Drag(Drag=3, Use Linear Drag=true, Use Rotational Drag=true)
   &SolveForcesAndVelocity(Clamp Velocity=false, Limit Acceleration=false)
   &ScaleColor(
    Scale Alpha=&FloatFromCurve(FloatCurve=DataInterface() {propertyValues {Curve {keys [
         {interpMode=RCIM_Cubic, tangentMode=RCTM_User, time=0, value=0}
         {interpMode=RCIM_Cubic, tangentMode=RCTM_User, time=0.5, value=1}
         {tangentMode=RCTM_User, time=1, value=0}
        ]}}})
    Scale RGB=&MakeVectorFromLinearColorRGB(Color=<rgba 1 0.7092 0.2159 1>)
    ScaleA=true
    ScaleRGB=true
   )
   /SetVariables(Particles.Glow03_Color=&MakeLinearColorFromVectorAndFloat(
     "Vector (RGB)"=&MakeVectorFromLinearColorRGB(Color=&MultiplyLinearColor(Linear Color=<rgba 1 0.1229 0 1>, Scale Factor=&MakeVector4_FromFloat(Float=1)))
     "Float (Alpha)"=&FloatFromCurve(FloatCurve=DataInterface() {propertyValues {Curve {keys [
          {interpMode=RCIM_Cubic, tangentMode=RCTM_User, time=0, value=0}
          {interpMode=RCIM_Cubic, tangentMode=RCTM_User, time=0.3, value=1}
          {arriveTangent=-0.0791, interpMode=RCIM_Cubic, leaveTangent=-0.0791, tangentMode=RCTM_User, time=0.6, value=1}
          {tangentMode=RCTM_User, time=1, value=0}
         ]}}})
    )) {isSetParameters=true}
   &ScaleSpriteSize(
    Scale Factor=&Multiply_Vector2D(
     A=&Multiply_Vector2D_ByFloat(Float=&RandomRangeFloat(Minimum=0.5, Maximum=1), Vector2D=<1 1>)
     B=$Particles.SizeCustomScale
    )
    Scale Sprite Size Mode=::NewEnumerator2
   )
   /SetVariables(Particles.Glow03_Size=&Multiply_Vector2D(A=<5 10>, B=$Particles.SizeCustomScale)) {isSetParameters=true}
  ]
  rendererClasses [&NiagaraSpriteRendererProperties]
  renderers [
   {rendererClass=&NiagaraSpriteRendererProperties}
   {rendererClass=&NiagaraSpriteRendererProperties, rendererIndex=1}
  ]
 }
 ParticleDecal001 {
  emitterUpdateScript [
   &EmitterState(
    Inactive Response=::NewEnumerator1
    Life Cycle Mode=::NewEnumerator1
    Loop Behavior=::NewEnumerator1
    Loop Duration Mode=::NewEnumerator1
    UseLoopDelay=false
   )
   &SpawnBurst_Instantaneous(Use Spawn Probability=true, UseLoopCountLimit=false)
  ]
  particleSpawnScript [
   &InitializeParticle(
    Mesh Scale Mode=::NewEnumerator3
    Mesh Uniform Scale=2
    Sprite Size Mode=::NewEnumerator3
    Uniform Sprite Size=25
    UsePositionOffset=false
    Write Material Random=true
   )
   &CameraOffset(Camera Offset Amount=30)
   &DynamicMaterialParameters(
    Index 0 Param 1=1
    Module.Index0Param0Write=true
    Module.Index0Param1Write=true
    Module.Index0Param2Write=true
    Module.Index0Param3Write=true
   )
  ]
  particleUpdateScript [
   &ParticleState(Kill Particles When Lifetime Has Elapsed=false)
   &ScaleColor(
    Scale Alpha=0.3
    Scale RGB=&MakeVectorFromLinearColorRGB(Color=<rgba 3 0.3 0 1>)
    ScaleA=true
    ScaleRGB=true
   )
   &FN_Light_Attributes(
    Color=<rgba 4 1.6 0 1>
    Create Color Attribute=true
    Create Exponent Attribute=true
    Create Position Attribute=true
    Create Radius Attribute=true
    Enable Particle Light Enable=true
    Exponent=2
    Position=<-25 -5 0>
    Radius=100
    Scale Color=2
   )
  ]
  rendererClasses [&NiagaraSpriteRendererProperties, &NiagaraLightRendererProperties]
  renderers [
   {rendererClass=&NiagaraSpriteRendererProperties}
   {rendererClass=&NiagaraSpriteRendererProperties, rendererIndex=1}
   {rendererClass=&NiagaraLightRendererProperties, rendererIndex=2}
  ]
 }
 Ring {
  emitterUpdateScript [
   &EmitterState(
    Inactive Response=::NewEnumerator1
    Life Cycle Mode=::NewEnumerator1
    Loop Behavior=::NewEnumerator1
    Loop Duration Mode=::NewEnumerator1
    UseLoopDelay=false
   )
   &SpawnBurst_Instantaneous(Spawn Count=2, Use Spawn Probability=true, UseLoopCountLimit=false)
  ]
  particleSpawnScript [
   &PartitionParticles()
   &InitializeParticle(
    Color Mode=::NewEnumerator0
    Lifetime=1.5
    Mesh Scale=&SelectVectorFromArray(
     Direct Array Index=$Particles.PartitionParticles.Partition
     Vector Selection Array=DataInterface() {propertyValues {FloatData [<2.6 2.6 2.6>, <2.6 -2.6 2.6>]}}
    )
    Mesh Scale Mode=::NewEnumerator4
    Position Offset=&Lerp_Vector(A=<-5 -5 0>, B=<12 1 -3>, Alpha=$User.On_BuiltIn)
    UsePositionOffset=true
    Write Material Random=true
   )
   &InitialMeshOrientation(
    Mesh Orientation Mode=::NewEnumerator2
    Rotation=&SelectVectorFromArray(
     Direct Array Index=$Particles.PartitionParticles.Partition
     Vector Selection Array=DataInterface() {propertyValues {FloatData [<0 0 0>, <0.25 0 0>]}}
    )
    Use Rotation Vector=true
   )
   &InitialMeshOrientation(
    Mesh Orientation Mode=::NewEnumerator2
    Rotation=&Lerp_Vector(A=<0 0 0>, B=<0 0.02 0.04>, Alpha=$User.On_BuiltIn)
    Use Rotation Vector=true
   ) {module=InitialMeshOrientation001}
   &SpriteFacingAndAlignment(
    Sprite Alignment=&Lerp_Vector(A=<1 0 0>, B=<1 0.25 -0.1>, Alpha=$User.On_BuiltIn)
    WriteAlignment=true
    WriteFacing=true
   )
   /SetVariables() {isSetParameters=true, module=SetVariables_#1}
  ]
  particleUpdateScript [
   &ParticleState(Kill Particles When Lifetime Has Elapsed=false, Let Infinitely Lived Particles Die When Emitter Deactivates=true)
   &ScaleMeshSize(Scale Factor=&Lerp_Vector(A=<1.1 1 1>, B=<1.2 1.05 1.05>, Alpha=$User.On_BuiltIn))
   &UpdateMeshOrientation(Rotation Rate=&SelectFloatFromArray(Direct Array Index=$Particles.PartitionParticles.Partition, Float Selection Array=DataInterface() {propertyValues {FloatData [1,-1]}}))
   &ScaleColor(
    Scale Alpha=&Multiply_Float(A=1, B=&FNTimeOfDayScalar(Night Scale=0.8, Morning Scale=1, Evening Scale=1))
    Scale RGB=&MakeVectorFromLinearColorRGB(Color=<rgba 0.9 0.9 0.9 1>)
    ScaleA=true
    ScaleRGB=true
   )
   &FN_Sprite_Attributes(
    Color=&Multiply_LinearColorByFloat(
     Linear Color=<rgba 1 0.1 0 1>
     Float=&Multiply_Float(
      A=&FNTimeOfDayScalar(Night Scale=0.8, Morning Scale=0.8, Evening Scale=0.8)
      B=&FN_FrontEndFloatModifier(Gameplay=1, Front End=0.7)
     )
    )
    Create Color Attribute=true
    Create Position Attribute=true
    Create SpriteSize Attribute=true
    Position:NiagaraPosition=&Lerp_Vector(A=<-20 -5 0>, B=<-7 -3 0>, Alpha=$User.On_BuiltIn)
    SpriteSize=<35 70>
    Write to Module Namespace=true
   )
  ]
  rendererClasses [&NiagaraMeshRendererProperties, &NiagaraSpriteRendererProperties]
  renderers [
   {rendererClass=&NiagaraMeshRendererProperties, rendererIndex=0}
   {rendererClass=&NiagaraSpriteRendererProperties, rendererIndex=1}
  ]
 }
 Ring001 {
  emitterUpdateScript [
   &EmitterState(
    Inactive Response=::NewEnumerator1
    Life Cycle Mode=::NewEnumerator1
    Loop Behavior=::NewEnumerator1
    Loop Duration Mode=::NewEnumerator1
    UseLoopDelay=false
   )
   &SpawnBurst_Instantaneous(Use Spawn Probability=true, UseLoopCountLimit=false)
  ]
  particleSpawnScript [
   &PartitionParticles()
   &InitializeParticle(
    Color Mode=::NewEnumerator0
    Lifetime=1.5
    Mesh Scale=&Lerp_Vector(A=<1 1 2.4>, B=<1 1 2.4>, Alpha=$User.On_BuiltIn)
    Mesh Scale Mode=::NewEnumerator4
    Position Offset=&Lerp_Vector(A=<0 -5 0>, B=<14 1 -1>, Alpha=$User.On_BuiltIn)
    Sprite Rotation Angle=0.25
    Sprite Rotation Mode=::NewEnumerator1
    Sprite Size=<60 30>
    Sprite Size Mode=::NewEnumerator4
    UsePositionOffset=true
    Write Material Random=true
   )
   &InitialMeshOrientation(
    Mesh Orientation Mode=::NewEnumerator2
    Rotation=&Lerp_Vector(A=<0 -0.25 0>, B=<0 -0.25 0.065>, Alpha=$User.On_BuiltIn)
    Use Rotation Vector=true
   )
   &InitialMeshOrientation(
    Mesh Orientation Mode=::NewEnumerator2
    Rotation=&MakeCustomVectorFromBool(
     Bool=$User.bIsFrontEndPreview
     True Vector=&Lerp_Vector(A=<0 0 0>, B=<-0.05 0 0>, Alpha=$User.On_BuiltIn)
     False Vector=&Lerp_Vector(A=<0 0 0>, B=<-0.03 0 0>, Alpha=$User.On_BuiltIn)
    )
    Use Rotation Vector=true
   ) {module=InitialMeshOrientation001}
   &SpriteFacingAndAlignment(Sprite Alignment=<1 0 0>, WriteAlignment=true, WriteFacing=true) {enabled=false}
   /SetVariables() {enabled=false, isSetParameters=true, module=SetVariables_#1}
   &CameraOffset(Camera Offset Amount=10)
  ]
  particleUpdateScript [
   &ParticleState(Kill Particles When Lifetime Has Elapsed=false, Let Infinitely Lived Particles Die When Emitter Deactivates=true)
   &ScaleColor(
    Scale Alpha=4
    Scale RGB=&MakeVectorFromLinearColorRGB(Color=<rgba 0.6562 0.4177 0.1678 1>)
    ScaleA=true
    ScaleRGB=true
   )
   &DynamicMaterialParameters(
    Index 0 Param 1=2
    Index 0 Param 2=2
    Index 0 Param 3=8.5
    Index 0 Param 4=1
    Module.Index0Float0Write=true
    Module.Index0Float1Write=true
    Module.Index0Float2Write=true
    Module.Index0Float3Write=true
   )
  ]
  rendererClasses [&NiagaraMeshRendererProperties]
  renderers [{rendererClass=&NiagaraMeshRendererProperties, rendererIndex=0}]
 }
 Trail {
  emitterUpdateScript [
   &EmitterState(
    Inactive Response=::NewEnumerator1
    Life Cycle Mode=::NewEnumerator1
    Loop Behavior=::NewEnumerator1
    Loop Duration Mode=::NewEnumerator1
    UseLoopDelay=false
   )
   &SpawnBurst_Instantaneous(Use Spawn Probability=true, UseLoopCountLimit=false)
  ]
  particleSpawnScript [
   &InitializeParticle(
    Color=<rgba 1 0.4644 0.25 2>
    Lifetime=5
    Mesh Scale=<1.6 1 1>
    Mesh Scale Mode=::NewEnumerator4
    Position Offset=<0 -7 0>
    UsePositionOffset=true
    Write Material Random=true
   )
   &InitialMeshOrientation(Mesh Orientation Mode=::NewEnumerator2, Rotation=<0 0.02 0.03>, Use Rotation Vector=true)
  ]
  particleUpdateScript [
   &ParticleState(Kill Particles When Lifetime Has Elapsed=false, Let Infinitely Lived Particles Die When Emitter Deactivates=true)
   &ScaleMeshSize(Scale Factor=&Lerp_Vector(A=<1 1 1>, B=<0 1 0>, Alpha=$User.On_BuiltIn))
   &DynamicMaterialParameters(
    Index 0 Param 1=&RandomRangeFloat(Minimum=1, Maximum=1.6)
    Index 0 Param 2=0.8
    Index 0 Param 3=1.5
    Index 0 Param 4=&RandomRangeFloat(Minimum=1, Maximum=1.6)
    Module.Index0Float0Write=true
    Module.Index0Float1Write=true
    Module.Index0Float2Write=true
    Module.Index0Float3Write=true
   )
  ]
  rendererClasses [&NiagaraMeshRendererProperties]
  renderers [{rendererClass=&NiagaraMeshRendererProperties, rendererIndex=0}]
 }
}
summary {
 systemName=NS_Pickaxe_EggnogFaucetUniversal_Idle
 emitters [
  {emitterName=Ring, rendererClasses [&NiagaraMeshRendererProperties, &NiagaraSpriteRendererProperties]}
  {emitterName=Center, rendererClasses [&NiagaraMeshRendererProperties, &NiagaraSpriteRendererProperties]}
  {emitterName=Particle001, rendererClasses [&NiagaraSpriteRendererProperties]}
  {emitterName=Ring001, rendererClasses [&NiagaraMeshRendererProperties]}
  {emitterName=FireRing_Limbs001, rendererClasses [&NiagaraMeshRendererProperties]}
  {emitterName=Fire_Particle001, rendererClasses [&NiagaraSpriteRendererProperties]}
  {
   emitterName=ParticleDecal001
   rendererClasses [&NiagaraSpriteRendererProperties, &NiagaraLightRendererProperties]
  }
  {emitterName=Trail, rendererClasses [&NiagaraMeshRendererProperties]}
 ]
 userVariables [
  User.bIsFrontEnd {type=NiagaraBool}
  User.bIsFrontEndPreview {type=NiagaraBool}
  User.On_BuiltIn {type=NiagaraFloat, defaultValue=0, description=""}
 ]
}
