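# Niagara swing trail for the EggnogFaucetUniversal pickaxe, sampled from its skeletal mesh: animation-trail ribbons, fire sprites and rising embers.
# Fire and ember emitters come in pairs, with User.On_BuiltIn crossfading spawn probability between sampling setups and narrowing ribbon widths.
# Trail_Dark layers three ribbon renderers beside Trail_Light; User.Swing02Delay switches both to delayed, emitter-age alpha fades.
# Front-end and time-of-day modifiers adjust size and opacity.
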NiagaraSystem /BRCosmetics/Weapons/FORT_Melee/Pickaxe_EggnogFaucetUniversal/FX/NS_Pickaxe_EggnogFaucetUniversal_Trail
dependencies {
 usedDataInterfaces [&NiagaraDataInterfaceCurve, &NiagaraDataInterfaceColorCurve, &NiagaraDataInterfaceSkeletalMesh]
 usedDynamicInputs []
 usedModules [
  &self:SystemSpawnScript.SystemScriptSource.SystemScriptGraph.NiagaraNodeAssignment_0.SetVariables_#1
  &self:SystemSpawnScript.SystemScriptSource.SystemScriptGraph.NiagaraNodeAssignment_1.SetVariables_#2
  &SystemState
 ]
}
emitters {
 Fire_Particle {
  emitterUpdateScript [
   &EmitterState(Life Cycle Mode=::NewEnumerator1, Loop Behavior=::NewEnumerator1, Loop Delay=0.05, Loop Duration=0.3, UseLoopDelay=true)
   &SpawnPerUnit(
    Max Movement Threshold=2500
    Module.Use Max Movement Threshold=true
    Module.Use Movement Tolerance=false
    Spawn Probability:NiagaraFloat=&Lerp_Float(A=1, B=0, Alpha=$User.On_BuiltIn)
    Spawn Spacing=15
    Use Spawn Probability=true
   )
  ]
  particleSpawnScript [
   &InitializeParticle(
    Lifetime Max=0.35
    Lifetime Min=0.3
    Lifetime Mode=::NewEnumerator1
    Sprite Rotation Mode=::NewEnumerator2
    Sprite Size Mode=::NewEnumerator1
    Uniform Sprite Size Max=100
    Uniform Sprite Size Min=50
    UsePositionOffset=true
    Write Material Random=true
   )
   &SkeletalMeshLocation(
    Bone Sampling Mode=::NewEnumerator4
    MaskSpawnGroup=false
    Sampled Position Offset=&Lerp_Vector(A=<-40 0 0>, B=<0 0 0>, Alpha=$User.On_BuiltIn)
    Skeletal Mesh=$System.Skeletal Mesh
   )
   &AddVelocity(Module.Use Falloff Distance=false, Override Seed=false, Velocity Mode=::NewEnumerator1, Velocity Speed=15)
   &InheritVelocity(
    Inherited Velocity Speed Limit=20
    Speed Limit=true
    UseSpeedThreshold=false
    Velocity Source=$Particles.SkeletalMeshLocation.SampledVelocity
   )
  ]
  particleUpdateScript [
   &ParticleState()
   /SetVariables(Particles.SubImageIndex=&FloatFromCurve(FloatCurve=DataInterface() {propertyValues {Curve {keys [<at 0 5>,<at 1 15>]}}})) {isSetParameters=true, module=SetVariables_#3}
   &ScaleSpriteSize(Scale Factor=&Vector2DFromFloat(Value=&FloatFromCurve(
      FloatCurve=DataInterface() {propertyValues {Curve {keys [
          {interpMode=RCIM_Cubic, leaveTangent=5.0519, tangentMode=RCTM_Break, time=0, value=0}
          {arriveTangent=1.0159, interpMode=RCIM_Cubic, leaveTangent=1.0159, tangentMode=RCTM_User, time=0.2, value=0.8}
          {arriveTangent=-1.4315, interpMode=RCIM_Cubic, tangentMode=RCTM_Break, time=1, value=0.5}
         ]}}}
      Scale Curve=&FN_FrontEndFloatModifier(Gameplay=0.8, Front End=0.6)
     )))
   &ScaleColor(
    Scale Alpha=&FloatFromCurve(
     FloatCurve=DataInterface() {propertyValues {Curve {keys [
         {arriveTangent=3.5924, interpMode=RCIM_Cubic, leaveTangent=3.5924, tangentMode=RCTM_User, time=0, value=0}
         {arriveTangent=-3.3587, interpMode=RCIM_Cubic, leaveTangent=-3.3587, tangentMode=RCTM_User, time=0.0493, value=0.989}
         {arriveTangent=-1.838, interpMode=RCIM_Cubic, leaveTangent=-1.838, tangentMode=RCTM_User, time=0.249, value=0.3945}
         <at 0.7 0 RCIM_Cubic>
        ]}}}
     Scale Curve=&Multiply_Float(A=0.4, B=&FNTimeOfDayScalar(Night Scale=0.8, Morning Scale=1, Evening Scale=0.8))
    )
    Scale RGB=&MakeVectorFromLinearColorRGB(Color=&ColorFromCurve(ColorCurve=DataInterface() {propertyValues {
        AlphaCurve {keys [<at 0 1 RCIM_Cubic>]}
        BlueCurve {keys [<at 0.0043 0>,<at 0.0811 0>,<at 0.4516 0.0081>]}
        GreenCurve {keys [<at 0.0043 0.5873>,<at 0.0811 0.1958>,<at 0.4516 0>]}
        RedCurve {keys [<at 0.0043 3>,<at 0.0811 1>,<at 0.4516 0.2917>]}
       }}))
    ScaleA=true
    ScaleRGB=true
   )
   &SolveForcesAndVelocity(Clamp Velocity=false, Limit Acceleration=false)
  ]
  rendererClasses [&NiagaraSpriteRendererProperties]
  renderers [{rendererClass=&NiagaraSpriteRendererProperties}]
 }
 Fire_Particle001 {
  emitterUpdateScript [
   &EmitterState(Life Cycle Mode=::NewEnumerator1, Loop Behavior=::NewEnumerator1, Loop Delay=0.05, Loop Duration=0.3, UseLoopDelay=true)
   &SpawnPerUnit(
    Max Movement Threshold=2500
    Module.Use Max Movement Threshold=true
    Module.Use Movement Tolerance=false
    Spawn Probability:NiagaraFloat=&Lerp_Float(A=0, B=1, Alpha=$User.On_BuiltIn)
    Spawn Spacing=15
    Use Spawn Probability=true
   )
  ]
  particleSpawnScript [
   &InitializeParticle(
    Lifetime Max=0.4
    Lifetime Min=0.3
    Lifetime Mode=::NewEnumerator1
    Sprite Rotation Mode=::NewEnumerator2
    Sprite Size Mode=::NewEnumerator1
    Uniform Sprite Size Max=100
    Uniform Sprite Size Min=50
    UsePositionOffset=true
    Write Material Random=true
   )
   &SkeletalMeshLocation(
    MaskSpawnGroup=false
    Mesh Sampling Type=::NewEnumerator5
    Sampled Position Offset=<-40 0 0>
    Skeletal Mesh=$System.Skeletal Mesh
    Socket Sampling Mode=::NewEnumerator5
   )
   &AddVelocity(Module.Use Falloff Distance=false, Override Seed=false, Velocity Mode=::NewEnumerator1, Velocity Speed=60)
   &InheritVelocity(
    Inherited Velocity Speed Limit=20
    Speed Limit=true
    UseSpeedThreshold=false
    Velocity Source=$Particles.SkeletalMeshLocation.SampledVelocity
   )
  ]
  particleUpdateScript [
   &ParticleState()
   /SetVariables(Particles.SubImageIndex=&FloatFromCurve(FloatCurve=DataInterface() {propertyValues {Curve {keys [<at 0 5>,<at 1 15>]}}})) {isSetParameters=true, module=SetVariables_#3}
   &ScaleSpriteSize(Scale Factor=&Vector2DFromFloat(Value=&FloatFromCurve(
      FloatCurve=DataInterface() {propertyValues {Curve {keys [
          {interpMode=RCIM_Cubic, leaveTangent=5.0519, tangentMode=RCTM_Break, time=0, value=0}
          {arriveTangent=1.0159, interpMode=RCIM_Cubic, leaveTangent=1.0159, tangentMode=RCTM_User, time=0.2, value=0.8}
          {arriveTangent=-1.4315, interpMode=RCIM_Cubic, tangentMode=RCTM_Break, time=1, value=0.5}
         ]}}}
      Scale Curve=&FN_FrontEndFloatModifier(Gameplay=0.8, Front End=0.6)
     )))
   &ScaleColor(
    Scale Alpha=&FloatFromCurve(
     FloatCurve=DataInterface() {propertyValues {Curve {keys [
         {arriveTangent=3.5924, interpMode=RCIM_Cubic, leaveTangent=3.5924, tangentMode=RCTM_User, time=0, value=0}
         {arriveTangent=-3.3587, interpMode=RCIM_Cubic, leaveTangent=-3.3587, tangentMode=RCTM_User, time=0.0493, value=0.989}
         {arriveTangent=-1.838, interpMode=RCIM_Cubic, leaveTangent=-1.838, tangentMode=RCTM_User, time=0.249, value=0.3945}
         <at 0.7 0 RCIM_Cubic>
        ]}}}
     Scale Curve=&Multiply_Float(A=0.4, B=&FNTimeOfDayScalar(Night Scale=0.8, Morning Scale=1, Evening Scale=0.8))
    )
    Scale RGB=&MakeVectorFromLinearColorRGB(Color=&ColorFromCurve(ColorCurve=DataInterface() {propertyValues {
        AlphaCurve {keys [<at 0 1 RCIM_Cubic>]}
        BlueCurve {keys [<at 0.0043 0>,<at 0.0811 0>,<at 0.4516 0.0081>]}
        GreenCurve {keys [<at 0.0043 0.3915>,<at 0.0811 0.1958>,<at 0.4516 0>]}
        RedCurve {keys [<at 0.0043 2>,<at 0.0811 1>,<at 0.4516 0.2917>]}
       }}))
    ScaleA=true
    ScaleRGB=true
   )
   &SolveForcesAndVelocity(Clamp Velocity=false, Limit Acceleration=false)
  ]
  rendererClasses [&NiagaraSpriteRendererProperties]
  renderers [{rendererClass=&NiagaraSpriteRendererProperties}]
 }
 Particle {
  emitterUpdateScript [
   &EmitterState(Life Cycle Mode=::NewEnumerator1, Loop Behavior=::NewEnumerator1, Loop Delay=0.05, Loop Duration=0.15, UseLoopDelay=true)
   &SpawnRate(Spawn Probability=&Lerp_Float(A=1, B=0, Alpha=$User.On_BuiltIn), SpawnRate=60, Use Spawn Probability=true)
  ]
  particleSpawnScript [
   &InitializeParticle(
    Lifetime Max=0.55
    Lifetime Min=0.35
    Lifetime Mode=::NewEnumerator1
    Mesh Scale Mode=::NewEnumerator1
    Mesh Uniform Scale Max=1.5
    Mesh Uniform Scale Min=0.5
    Sprite Size Max=<2 6>
    Sprite Size Min=<1 2>
    Sprite Size Mode=::NewEnumerator2
    UsePositionOffset=true
    Write Material Random=true
   )
   &SkeletalMeshLocation(
    MaskSpawnGroup=false
    Mesh Sampling Type=::NewEnumerator5
    Sampled Position Offset=<0 20 0>
    Skeletal Mesh=$System.Skeletal Mesh
    Socket Sampling Mode=::NewEnumerator5
   )
   &ShapeLocation(
    MaskSpawnGroup=false
    Offset Mode=::NewEnumerator0
    Override Seed=false
    Rotation Axis=<0 1 0>
    Rotation Mode=::NewEnumerator1
    Sphere Radius=10
    UseOwnerScale=true
   )
   &AddVelocity(Velocity=&RandomRangeVector(Minimum=<0 0 80>, Maximum=<0 0 100>))
  ]
  particleUpdateScript [
   &ParticleState()
   &ScaleSpriteSize(
    Uniform Curve Scale=0.8
    Uniform Curve Sprite Scale=DataInterface() {propertyValues {Curve {keys [<at 0 0 RCIM_Cubic>,<at 0.2 1 RCIM_Cubic>,<at 1 0 RCIM_Cubic>]}}}
   )
   /SetVariables(Particles.Glow03_Size=&Multiply_Vector2D_ByFloat(
     Float=&FloatFromCurve(
      FloatCurve=DataInterface() {propertyValues {Curve {keys [
          <at 0 0 RCIM_Cubic>
          <at 0.2 1.3 RCIM_Cubic>
          {arriveTangent=-0.1423, interpMode=RCIM_Cubic, leaveTangent=-0.1423, tangentMode=RCTM_User, time=1, value=0}
         ]}}}
      Scale Curve=8
     )
     Vector2D=<0.7 1.5>
    )) {isSetParameters=true, module=SetVariables_#4}
   &ScaleColor(
    Scale Alpha=&Multiply_Float(
     A=1
     B=&FloatFromCurve(FloatCurve=DataInterface() {propertyValues {Curve {keys [
          {interpMode=RCIM_Cubic, tangentMode=RCTM_User, time=0.1, value=0}
          {arriveTangent=-1.3622, interpMode=RCIM_Cubic, leaveTangent=-1.3622, tangentMode=RCTM_User, time=0.2, value=1}
          {tangentMode=RCTM_User, time=0.8, value=0}
         ]}}})
    )
    Scale RGB=&MakeVectorFromLinearColorRGB(Color=&ColorFromCurve(ColorCurve=DataInterface() {propertyValues {
        AlphaCurve {keys [<at 0.0926 1 RCIM_Cubic>,<at 1 0 RCIM_Cubic>]}
        BlueCurve {keys [<at 0 0.1336 RCIM_Cubic>]}
        GreenCurve {keys [<at 0 0.5819 RCIM_Cubic>]}
        RedCurve {keys [<at 0 5 RCIM_Cubic>]}
       }}))
    ScaleA=true
    ScaleRGB=true
   )
   /SetVariables(Particles.Glow03_Color=&MakeLinearColorFromVectorAndFloat(
     "Vector (RGB)"=&MakeVectorFromLinearColorRGB(Color=&MultiplyLinearColor(Linear Color=<rgba 1.5 0.25 0 1>, Scale Factor=&MakeVector4_FromFloat(Float=1)))
     "Float (Alpha)"=&Multiply_Float(A=0.2, B=1)
    )) {isSetParameters=true, module=SetVariables_#5}
   &GravityForce(Gravity=&RandomRangeVector(Minimum=<0 0 50>, Maximum=<0 0 -100>))
   &CurlNoiseForce(
    Noise Frequency=300
    Noise Strength=&RandomRangeFloat(Minimum=1000, Maximum=800)
    Pan Noise=true
    Pan Noise Field=<0 0 1>
   )
   &Drag(Drag=3, Use Linear Drag=true, Use Rotational Drag=true)
   &SolveForcesAndVelocity(Clamp Velocity=false, Limit Acceleration=false)
  ]
  rendererClasses [&NiagaraSpriteRendererProperties]
  renderers [
   {rendererClass=&NiagaraSpriteRendererProperties}
   {rendererClass=&NiagaraSpriteRendererProperties, rendererIndex=1}
  ]
 }
 Particle001 {
  emitterUpdateScript [
   &EmitterState(Life Cycle Mode=::NewEnumerator1, Loop Behavior=::NewEnumerator1, Loop Delay=0.05, Loop Duration=0.15, UseLoopDelay=true)
   &SpawnRate(Spawn Probability=&Lerp_Float(A=0, B=1, Alpha=$User.On_BuiltIn), SpawnRate=60, Use Spawn Probability=true)
  ]
  particleSpawnScript [
   &InitializeParticle(
    Lifetime Max=0.55
    Lifetime Min=0.35
    Lifetime Mode=::NewEnumerator1
    Mesh Scale Mode=::NewEnumerator1
    Mesh Uniform Scale Max=1.5
    Mesh Uniform Scale Min=0.5
    Sprite Size Max=<2 6>
    Sprite Size Min=<1 2>
    Sprite Size Mode=::NewEnumerator2
    UsePositionOffset=true
    Write Material Random=true
   )
   &SkeletalMeshLocation(
    MaskSpawnGroup=false
    Mesh Sampling Type=::NewEnumerator5
    Sampled Position Offset=<-40 0 0>
    Skeletal Mesh=$System.Skeletal Mesh
    Socket Sampling Mode=::NewEnumerator5
   )
   &ShapeLocation(
    MaskSpawnGroup=false
    Offset Mode=::NewEnumerator0
    Override Seed=false
    Rotation Axis=<0 1 0>
    Rotation Mode=::NewEnumerator1
    Sphere Radius=13
    UseOwnerScale=true
   )
   &AddVelocity(Velocity=&RandomRangeVector(Minimum=<0 0 80>, Maximum=<0 0 100>), Velocity Speed Scale=1.5)
  ]
  particleUpdateScript [
   &ParticleState()
   &ScaleSpriteSize(
    Uniform Curve Scale=0.8
    Uniform Curve Sprite Scale=DataInterface() {propertyValues {Curve {keys [<at 0 0 RCIM_Cubic>,<at 0.2 1 RCIM_Cubic>,<at 1 0 RCIM_Cubic>]}}}
   )
   /SetVariables(Particles.Glow03_Size=&Multiply_Vector2D_ByFloat(
     Float=&FloatFromCurve(
      FloatCurve=DataInterface() {propertyValues {Curve {keys [
          <at 0 0 RCIM_Cubic>
          <at 0.2 1.3 RCIM_Cubic>
          {arriveTangent=-0.1423, interpMode=RCIM_Cubic, leaveTangent=-0.1423, tangentMode=RCTM_User, time=1, value=0}
         ]}}}
      Scale Curve=8
     )
     Vector2D=<0.7 1.5>
    )) {isSetParameters=true, module=SetVariables_#4}
   &ScaleColor(
    Scale Alpha=&Multiply_Float(
     A=1
     B=&FloatFromCurve(FloatCurve=DataInterface() {propertyValues {Curve {keys [
          {interpMode=RCIM_Cubic, tangentMode=RCTM_User, time=0.1, value=0}
          {arriveTangent=-1.3622, interpMode=RCIM_Cubic, leaveTangent=-1.3622, tangentMode=RCTM_User, time=0.2, value=1}
          {tangentMode=RCTM_User, time=0.8, value=0}
         ]}}})
    )
    Scale RGB=&MakeVectorFromLinearColorRGB(Color=&ColorFromCurve(ColorCurve=DataInterface() {propertyValues {
        AlphaCurve {keys [<at 0.0926 1 RCIM_Cubic>,<at 1 0 RCIM_Cubic>]}
        BlueCurve {keys [<at 0 0.1336 RCIM_Cubic>]}
        GreenCurve {keys [<at 0 0.5819 RCIM_Cubic>]}
        RedCurve {keys [<at 0 5 RCIM_Cubic>]}
       }}))
    ScaleA=true
    ScaleRGB=true
   )
   /SetVariables(Particles.Glow03_Color=&MakeLinearColorFromVectorAndFloat(
     "Vector (RGB)"=&MakeVectorFromLinearColorRGB(Color=&MultiplyLinearColor(Linear Color=<rgba 1.5 0.25 0 1>, Scale Factor=&MakeVector4_FromFloat(Float=1)))
     "Float (Alpha)"=&Multiply_Float(A=0.2, B=1)
    )) {isSetParameters=true, module=SetVariables_#5}
   &GravityForce(Gravity=&RandomRangeVector(Minimum=<0 0 50>, Maximum=<0 0 -100>))
   &CurlNoiseForce(
    Noise Frequency=300
    Noise Strength=&RandomRangeFloat(Minimum=1000, Maximum=800)
    Pan Noise=true
    Pan Noise Field=<0 0 1>
   )
   &Drag(Drag=3, Use Linear Drag=true, Use Rotational Drag=true)
   &SolveForcesAndVelocity(Clamp Velocity=false, Limit Acceleration=false)
  ]
  rendererClasses [&NiagaraSpriteRendererProperties]
  renderers [
   {rendererClass=&NiagaraSpriteRendererProperties}
   {rendererClass=&NiagaraSpriteRendererProperties, rendererIndex=1}
  ]
 }
 Trail_Dark {
  emitterUpdateScript [
   &EmitterState(
    Life Cycle Mode=::NewEnumerator1
    Loop Behavior=::NewEnumerator1
    Loop Delay=&MakeCustomFloatFromBool(Bool=$User.Swing02Delay, True Float=0, False Float=0)
    Loop Duration=0.8
    UseLoopDelay=true
   )
   &FN_SpawnAnimTrail()
  ]
  particleSpawnScript [
   &FN_InitializeAnimTrailBoneIndecies()
   &FN_InitializeAnimTrail(
    AnimTrailWidth=&Lerp_Float(A=1.4, B=0.6, Alpha=$User.On_BuiltIn)
    Color=<rgba 1 1 1 0.8>
    Lifetime=0.4
    SkeletalMesh=$System.Skeletal Mesh
   )
  ]
  particleUpdateScript [
   &ParticleState()
   &FN_UpdateAnimTrail(SkeletalMesh=$System.Skeletal Mesh)
   &ScaleColor(
    Scale Alpha=&MakeCustomFloatFromBool(
     Bool=$User.Swing02Delay
     True Float=&FloatFromCurve(
      FloatCurve=DataInterface() {propertyValues {Curve {keys [
          {arriveTangent=0.0425, interpMode=RCIM_Cubic, leaveTangent=0.0425, tangentMode=RCTM_User, time=0.13, value=0}
          {arriveTangent=-1.7479, interpMode=RCIM_Cubic, leaveTangent=-1.7479, tangentMode=RCTM_User, time=0.15, value=0.85}
          {arriveTangent=0.0495, interpMode=RCIM_Cubic, leaveTangent=0.0495, tangentMode=RCTM_User, time=0.4, value=0}
         ]}}}
      CurveIndex=$Emitter.Age
      Scale Curve=&MakeCustomFloatFromBool(Bool=$User.bIsFrontEndPreview, True Float=3, False Float=5)
     )
     False Float=&FloatFromCurve(
      FloatCurve=DataInterface() {propertyValues {Curve {keys [
          {arriveTangent=-1.6719, interpMode=RCIM_Cubic, leaveTangent=-1.6719, tangentMode=RCTM_User, time=0, value=1}
          {arriveTangent=0.0495, interpMode=RCIM_Cubic, leaveTangent=0.0495, tangentMode=RCTM_User, time=1, value=0}
         ]}}}
      Scale Curve=&MakeCustomFloatFromBool(Bool=$User.bIsFrontEndPreview, True Float=3, False Float=5)
     )
    )
    Scale RGB=&MakeVectorFromLinearColorRGB(Color=&ColorFromCurve(ColorCurve=DataInterface() {propertyValues {
        AlphaCurve {keys []}
        BlueCurve {keys [<at 0.6209 0.0017 RCIM_Cubic>]}
        GreenCurve {keys [<at 0.6209 0.0017 RCIM_Cubic>]}
        RedCurve {keys [<at 0.6209 0.0417 RCIM_Cubic>]}
       }}))
    ScaleA=true
    ScaleRGB=true
   )
   &ScaleRibbonWidth(Ribbon Width Scale=&FloatFromCurve(
     FloatCurve=DataInterface() {propertyValues {Curve {keys [{arriveTangent=-0.0514, time=0, value=1}, {arriveTangent=-1.0387, leaveTangent=-1.0387, time=1, value=0.6}]}}}
     Scale Curve=&FN_FrontEndFloatModifier(Gameplay=1, Front End=0.8)
    ))
   /SetVariables(Particles.Color002=&MakeLinearColorFromVectorAndFloat(
     "Vector (RGB)"=&MakeVectorFromLinearColorRGB(Color=<rgba 1 1 1 1>)
     "Float (Alpha)"=&MakeCustomFloatFromBool(
      Bool=$User.Swing02Delay
      True Float=&FloatFromCurve(
       FloatCurve=DataInterface() {propertyValues {Curve {keys [
           {arriveTangent=-0.014, interpMode=RCIM_Cubic, leaveTangent=-0.014, tangentMode=RCTM_User, time=0.13, value=0}
           {arriveTangent=-1.7131, interpMode=RCIM_Cubic, leaveTangent=-1.7131, tangentMode=RCTM_User, time=0.15, value=0.7978}
           {arriveTangent=-0.0168, interpMode=RCIM_Cubic, tangentMode=RCTM_Break, time=0.7, value=0}
          ]}}}
       CurveIndex=$Emitter.Age
       Scale Curve=&Multiply_Float(
        A=&MakeCustomFloatFromBool(Bool=$User.bIsFrontEndPreview, True Float=0.8, False Float=1)
        B=&FNTimeOfDayScalar(Night Scale=0.5, Morning Scale=0.6, Evening Scale=0.6)
       )
      )
      False Float=&FloatFromCurve(
       FloatCurve=DataInterface() {propertyValues {Curve {keys [
           {arriveTangent=-2.5375, interpMode=RCIM_Cubic, leaveTangent=-2.5375, tangentMode=RCTM_User, time=0, value=1}
           {arriveTangent=-1.7131, interpMode=RCIM_Cubic, leaveTangent=-1.7131, tangentMode=RCTM_User, time=0.2547, value=0.4}
           {arriveTangent=-0.0168, interpMode=RCIM_Cubic, tangentMode=RCTM_Break, time=0.7, value=0}
          ]}}}
       Scale Curve=&Multiply_Float(
        A=&MakeCustomFloatFromBool(Bool=$User.bIsFrontEndPreview, True Float=0.8, False Float=1)
        B=&FNTimeOfDayScalar(Night Scale=0.5, Morning Scale=0.6, Evening Scale=0.6)
       )
      )
     )
    )) {isSetParameters=true}
   /SetVariables(Particles.Color003:LinearColor=&MakeLinearColorFromVectorAndFloat(
     "Vector (RGB)"=&MakeVectorFromLinearColorRGB(Color=<rgba 0.6667 0.0133 0.0179 1>)
     "Float (Alpha)"=&MakeCustomFloatFromBool(
      Bool=$User.Swing02Delay
      True Float=&Multiply_Float(
       A=&FloatFromCurve(
        FloatCurve=DataInterface() {propertyValues {Curve {keys [
            {arriveTangent=-0.0654, interpMode=RCIM_Cubic, leaveTangent=-0.0654, tangentMode=RCTM_User, time=0.13, value=0}
            {arriveTangent=-2.1619, interpMode=RCIM_Cubic, leaveTangent=-2.1619, tangentMode=RCTM_User, time=0.15, value=0.8}
            {arriveTangent=-0.0464, interpMode=RCIM_Cubic, tangentMode=RCTM_Break, time=0.4, value=0}
           ]}}}
        CurveIndex=$Emitter.Age
       )
       B=&FloatFromCurve(
        FloatCurve=DataInterface() {propertyValues {Curve {keys [<at 0.5 1>,<at 0.8 0>]}}}
        CurveIndex=$Emitter.Age
        Scale Curve=&Multiply_Float(
         A=&FNTimeOfDayScalar(Night Scale=0.5, Morning Scale=0.6, Evening Scale=0.6)
         B=&MakeCustomFloatFromBool(Bool=$User.bIsFrontEndPreview, True Float=0.8, False Float=1.3)
        )
       )
      )
      False Float=&Multiply_Float(
       A=&FloatFromCurve(FloatCurve=DataInterface() {propertyValues {Curve {keys [
            {arriveTangent=-1.8936, interpMode=RCIM_Cubic, leaveTangent=-1.8936, tangentMode=RCTM_User, time=0, value=1}
            {arriveTangent=-2.1619, interpMode=RCIM_Cubic, leaveTangent=-2.1619, tangentMode=RCTM_User, time=0.2163, value=0.3648}
            {arriveTangent=-0.0464, interpMode=RCIM_Cubic, tangentMode=RCTM_Break, time=0.6, value=0}
           ]}}})
       B=&FloatFromCurve(
        FloatCurve=DataInterface() {propertyValues {Curve {keys [<at 0.5 1>,<at 0.8 0>]}}}
        CurveIndex=$Emitter.Age
        Scale Curve=&Multiply_Float(
         A=&FNTimeOfDayScalar(Night Scale=0.5, Morning Scale=0.6, Evening Scale=0.6)
         B=&MakeCustomFloatFromBool(Bool=$User.bIsFrontEndPreview, True Float=0.8, False Float=1.3)
        )
       )
      )
     )
    )) {isSetParameters=true}
   /SetVariables(Particles.Width_Red:NiagaraFloat=&Multiply_Float(
     A=&FloatFromCurve(FloatCurve=DataInterface() {propertyValues {Curve {keys [{arriveTangent=-0.0514, time=0, value=1}, {arriveTangent=-1.0387, leaveTangent=-1.0387, time=1, value=0.5}]}}})
     B=35
    )) {isSetParameters=true}
   /SetVariables(Particles.Dyn_Red:Vector4f=&MakeVector4(
     X=&FloatFromCurve(FloatCurve=DataInterface() {propertyValues {Curve {keys [{arriveTangent=-10.0187, interpMode=RCIM_Cubic, leaveTangent=-10.0187, tangentMode=RCTM_User, time=0, value=2.5}, <at 0.5 -20 RCIM_Cubic>]}}})
     Y=0
     Z=0
     W=0
    )) {isSetParameters=true}
  ]
  rendererClasses [&NiagaraRibbonRendererProperties]
  renderers [
   {rendererClass=&NiagaraRibbonRendererProperties}
   {rendererClass=&NiagaraRibbonRendererProperties, rendererIndex=1}
   {rendererClass=&NiagaraRibbonRendererProperties, rendererIndex=2}
  ]
 }
 Trail_Light {
  emitterUpdateScript [
   &EmitterState(Life Cycle Mode=::NewEnumerator1, Loop Behavior=::NewEnumerator1, Loop Duration=0.8, UseLoopDelay=false)
   &FN_SpawnAnimTrail()
  ]
  particleSpawnScript [
   &FN_InitializeAnimTrailBoneIndecies()
   &FN_InitializeAnimTrail(AnimTrailWidth=&Lerp_Float(A=2, B=1, Alpha=$User.On_BuiltIn), Lifetime=0.28, SkeletalMesh=$System.Skeletal Mesh)
  ]
  particleUpdateScript [
   &ParticleState()
   &FN_UpdateAnimTrail(SkeletalMesh=$System.Skeletal Mesh)
   &ScaleColor(
    Scale Alpha=&MakeCustomFloatFromBool(
     Bool=$User.Swing02Delay
     True Float=&FloatFromCurve(
      FloatCurve=DataInterface() {propertyValues {Curve {keys [
          {arriveTangent=-0.0182, interpMode=RCIM_Cubic, leaveTangent=-0.0182, tangentMode=RCTM_User, time=0.13, value=0}
          {arriveTangent=-0.8624, interpMode=RCIM_Cubic, leaveTangent=-0.8624, tangentMode=RCTM_User, time=0.15, value=0.6193}
          {arriveTangent=-0.3295, interpMode=RCIM_Cubic, leaveTangent=-0.3295, tangentMode=RCTM_User, time=1, value=0}
         ]}}}
      CurveIndex=$Emitter.Age
     )
     False Float=&FloatFromCurve(FloatCurve=DataInterface() {propertyValues {Curve {keys [
          {arriveTangent=-1.6719, interpMode=RCIM_Cubic, leaveTangent=-1.6719, tangentMode=RCTM_User, time=0, value=1}
          {arriveTangent=0.0495, interpMode=RCIM_Cubic, leaveTangent=0.0495, tangentMode=RCTM_User, time=1, value=0}
         ]}}})
    )
    Scale RGB=&MakeVectorFromLinearColorRGB(Color=&ColorFromCurve(ColorCurve=DataInterface() {propertyValues {
        AlphaCurve {keys []}
        BlueCurve {keys [<at 0.0344 0.7344>,<at 0.5408 0.5729 RCIM_Cubic>]}
        GreenCurve {keys [<at 0.0344 0.8207>,<at 0.5408 0.7117 RCIM_Cubic>]}
        RedCurve {keys [<at 0.0344 1>,<at 0.5408 1 RCIM_Cubic>]}
       }}))
    ScaleA=true
    ScaleRGB=true
   )
   &ScaleRibbonWidth(Ribbon Width Scale=&FloatFromCurve(
     FloatCurve=DataInterface() {propertyValues {Curve {keys [{arriveTangent=-0.0514, time=0, value=1}, {arriveTangent=-1.0387, leaveTangent=-1.0387, time=1, value=0.6}]}}}
     Scale Curve=&FN_FrontEndFloatModifier(Gameplay=1, Front End=0.5)
    ))
  ]
  rendererClasses [&NiagaraRibbonRendererProperties]
  renderers [{rendererClass=&NiagaraRibbonRendererProperties}]
 }
}
summary {
 systemName=NS_Pickaxe_EggnogFaucetUniversal_Trail
 emitters [
  {emitterName=Particle, rendererClasses [&NiagaraSpriteRendererProperties]}
  {emitterName=Fire_Particle, rendererClasses [&NiagaraSpriteRendererProperties]}
  {emitterName=Trail_Dark, rendererClasses [&NiagaraRibbonRendererProperties]}
  {emitterName=Trail_Light, rendererClasses [&NiagaraRibbonRendererProperties]}
  {emitterName=Fire_Particle001, rendererClasses [&NiagaraSpriteRendererProperties]}
  {emitterName=Particle001, rendererClasses [&NiagaraSpriteRendererProperties]}
 ]
 userVariables [
  User.bIsFrontEnd {type=NiagaraBool}
  User.bIsFrontEndPreview {type=NiagaraBool}
  User.On_BuiltIn {type=NiagaraFloat, defaultValue=0, description=""}
  User.RibbonIDOffset {type=NiagaraInt32}
  User.Swing02Delay {type=NiagaraBool}
 ]
}
